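; Gameplay ability for a Season 16 Battle Pass quest that credits the player for bringing a pickup or semi truck to a tagged POI.
; On vehicle entry, it checks the vehicle's tags and, if the player lacks the POI tag, waits for that tag to be added.
; When it is added, the ability rechecks the vehicle, sends the quest stat event and ends.
; Exiting the vehicle also ends the ability.
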(event Ability|EventActivateAbilityFromEvent
  (params (EventData :type Gameplay%20Event%20Data%20Structure%20%28by%20ref%29))
  (bind (_case_0_true _case_1_true) (GameplayTags|TagContainer|CompareTagAssetInterfacetoOtherTags (Variables|Default|GetT_Car_BasicPickup) (Variables|Default|GetT_Car_Semi) (Vehicle|GetVehicleAbilitySystemComponent (Vehicle|GetVehicleInterface (Variables|GAVars|Data|GetOwningPlayerPawn)))))
  (bind (_case_0_true_2 _case_1_true_2) (GameplayTags|TagContainer|CompareTagAssetInterfacetoOtherTags (Variables|Default|GetT_Car_BasicPickup) (Variables|Default|GetT_Car_Semi) (Vehicle|GetVehicleAbilitySystemComponent (Vehicle|GetVehicleInterface (Variables|GAVars|Data|GetOwningPlayerPawn)))))
  (bind _owningplayerpawn (Variables|GAVars|Data|GetOwningPlayerPawn))
  (|CallParentFunction EventData)
  (switch Utilities|FlowControl|Switch|SwitchonGameplayTag (Utilities|Struct|BreakGameplayEventData EventData)
    (:Event.Vehicle.Entered
      ;[1 1 1] Check Tags
      (if (or _case_0_true _case_1_true)
        (if (not (GameplayTags|TagContainer|CompareTagAssetInterfacetoOtherTags (Variables|Default|GetPOI%20Tag) _owningplayerpawn))
          (Ability|Tasks|WaitGameplayTagAdd (Variables|Default|GetPOI%20Tag) 0 true
            (:Added
              (if (or _case_0_true_2 _case_1_true_2)
                (Quest|SendComplexCustomStatEvent (Variables|GAVars|Data|GetOwningPlayerQuestManager) 0 (Variables|Default|GetT_Quest) "(GameplayTags=)" 1)
                (else
                  _))
              (Ability|EndAbility))))
        (else
          _))
      (Ability|EndAbility))
    (:Event.Vehicle.Exit
      (Ability|EndAbility))))
